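# Niagara system for the Stars vehicle boost cosmetic, emitting star sprites and ribbon trails backward from skeletal mesh sockets, tinted by painted color variant.
# Star deaths write to a position array; system-level ReadDeath apparently feeds a secondary burst of small spinning sprites.
# Two ribbon emitters run, one curl-noised.
# GameModeID switches velocities, rates and lifetimes; source-movement inheritance fades over particle life, and four near-duplicate emitters are disabled.
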NiagaraSystem /VehicleCosmetics/Boosts/Boost_Stars/FX/NS_Boost_Stars
dependencies {
 usedDataInterfaces [
  &NiagaraDataInterfaceCurlNoise
  &NiagaraDataInterfaceCurve
  &NiagaraDataInterfaceSpriteRendererInfo
  &NiagaraDataInterfaceArrayFloat3
  &NiagaraDataInterfaceSkeletalMesh
  &NiagaraDataInterfaceArrayPosition
 ]
 usedDynamicInputs []
 usedModules [
  &self:SystemSpawnScript.SystemScriptSource.SystemScriptGraph.NiagaraNodeAssignment_0.SetVariables_#1
  null
  &self:SystemSpawnScript.SystemScriptSource.SystemScriptGraph.NiagaraNodeAssignment_2.SetVariables_#2
  &SystemState
  &self:ReadDeath
  &self:SystemSpawnScript.SystemScriptSource.SystemScriptGraph.NiagaraNodeAssignment_1.SetVariables_#3
  &self:SystemSpawnScript.SystemScriptSource.SystemScriptGraph.NiagaraNodeAssignment_3.SetVariables_#4
 ]
}
emitters {
 NE_VC_Boost_Sprite_Parent {
  bEnabled=false
  emitterUpdateScript [
   &EmitterState()
   &SpawnPerFrame(Spawn:NiagaraBool=$System.DeathSpawn, Spawn Count=8, Use Spawn Probability=false)
   &VC_FilteredSocketCount() {enabled=false}
   &SpawnRate(SpawnRate=50, Use Spawn Probability=false) {enabled=false}
  ]
  particleSpawnScript [
   &VC_PaintedVariantsOverride(
    Black=<rgba 0 0 0 1>
    BurntSienna=<rgba 1.4 0.3939 0.124 1>
    Cobalt=<rgba 0 0.148 8 1>
    Crimson=<rgba 5 0 0.1355 1>
    DefaultColor=<rgba 6 4.8 0.6 1>
    ForestGreen=<rgba 0 5 0.2526 1>
    Grey=<rgba 0.1927 0.1927 0.1927 1>
    Lime=<rgba 3 6 0 1>
    Orange=<rgba 5 0.6667 0 1>
    Pink=<rgba 2 0.2396 1.3598 1>
    Purple=<rgba 0.6156 0 1.5 1>
    SkyBlue=<rgba 0 3.3269 5 1>
    TitaniumWhite=<rgba 4 4 4 1>
   )
   &InitializeParticle(
    Color Mode=::NewEnumerator0
    Lifetime=&RandomRangeFloat(Minimum=0.5, Maximum=0.2)
    Mass Mode=::NewEnumerator3
    Position=$System.DeathPosition
    Position Mode=::NewEnumerator1
    Sprite Size Mode=::NewEnumerator1
    Uniform Sprite Size Max=&Multiply_Float(A=10, B=2)
    Uniform Sprite Size Min=&Multiply_Float(A=8, B=2)
    UsePositionOffset=false
    Write Material Random=true
   )
   &PartitionParticles(Number Of Partitions=$Emitter.PortCount) {enabled=false}
   &SkeletalMeshLocation(
    Bone / Socket Index=$Particles.PartitionParticles.Partition
    Bone / Socket Sampling Mode=::NewEnumerator5
    MaskSpawnGroup=false
    Mesh Sampling Type=::NewEnumerator6
    Skeletal Mesh=$System.SkeletalMesh
   ) {enabled=false}
   &AddVelocity(Velocity=&MergeHelper_VectorSwitch(
     0=&RandomRangeVector(Minimum=<-700 200 100>, Maximum=<-1000 -200 300>)
     1=&RandomRangeVector(Minimum=<-1000 200 300>, Maximum=<-1500 -200 200>)
     2=?
     3=?
     Index=$User.GameModeID
     Num=::NewEnumerator0
    ))
   &ShapeLocation(MaskSpawnGroup=false, Override Seed=false, Sphere Radius=4, UseOwnerScale=true)
  ]
  particleUpdateScript [
   &ParticleState()
   &SpriteRotationRate(Rotation Rate=&RandomRangeFloat(Minimum=350, Maximum=500))
   &AccelerationForce(Acceleration=&RandomRangeVector(Minimum=<0 0 -256>, Maximum=<0 0 -128>))
   &Collision(
    Age Colliding Particles=false
    Friction:NiagaraFloat=&RandomRangeFloat(Minimum=0.2, Maximum=0.5)
    KillOnCollision=false
    Module.Control Rest State Via Surface Normal=false
    Module.EnableMaxCollisionCount=false
    Module.Randomize Collision Normal Vector=false
   )
   &GravityForce()
   &SolveForcesAndVelocity(Clamp Velocity=false, Limit Acceleration=false)
   &SubUVAnimation(
    "Sprite Renderer "=DataInterface() {propertyValues {SpriteRenderer=&self:NE_VC_Boost_Sprite_Parent_0.NiagaraSpriteRendererProperties_3}}
    UseEndFrame=false
    UseStartFrame=false
   )
  ]
  rendererClasses [&NiagaraSpriteRendererProperties]
  renderers [{rendererClass=&NiagaraSpriteRendererProperties}]
 }
 NE_VC_Boost_Sprite_Parent001 {
  emitterUpdateScript [
   &EmitterState()
   &SpawnPerFrame(Spawn:NiagaraBool=$System.DeathSpawn, Spawn Count=8, Use Spawn Probability=false)
   &VC_FilteredSocketCount() {enabled=false}
   &SpawnRate(SpawnRate=50, Use Spawn Probability=false) {enabled=false}
  ]
  particleSpawnScript [
   &VC_PaintedVariantsOverride(
    Black=<rgba 0 0 0 1>
    BurntSienna=<rgba 1.4 0.3939 0.124 1>
    Cobalt=<rgba 0 0.148 8 1>
    Crimson=<rgba 5 0 0.1355 1>
    DefaultColor=<rgba 6 4.8 0.6 1>
    ForestGreen=<rgba 0 5 0.2526 1>
    Grey=<rgba 0.1927 0.1927 0.1927 1>
    Lime=<rgba 3 6 0 1>
    Orange=<rgba 5 0.6667 0 1>
    Pink=<rgba 2 0.2396 1.3598 1>
    Purple=<rgba 0.6156 0 1.5 1>
    SkyBlue=<rgba 0 3.3269 5 1>
    TitaniumWhite=<rgba 4 4 4 1>
   )
   &InitializeParticle(
    Color Mode=::NewEnumerator0
    Lifetime=&RandomRangeFloat(Minimum=0.5, Maximum=0.2)
    Mass Mode=::NewEnumerator3
    Position=$System.DeathPosition
    Position Mode=::NewEnumerator1
    Sprite Size Mode=::NewEnumerator1
    Uniform Sprite Size Max=&Multiply_Float(A=10, B=2)
    Uniform Sprite Size Min=&Multiply_Float(A=8, B=2)
    UsePositionOffset=false
    Write Material Random=true
   )
   &PartitionParticles(Number Of Partitions=$Emitter.PortCount) {enabled=false}
   &SkeletalMeshLocation(
    Bone / Socket Index=$Particles.PartitionParticles.Partition
    Bone / Socket Sampling Mode=::NewEnumerator5
    MaskSpawnGroup=false
    Mesh Sampling Type=::NewEnumerator6
    Skeletal Mesh=$System.SkeletalMesh
   ) {enabled=false}
   &AddVelocity(Velocity=&MergeHelper_VectorSwitch(
     0=&RandomRangeVector(Minimum=<-700 200 100>, Maximum=<-1000 -200 300>)
     1=&RandomRangeVector(Minimum=<-1000 200 300>, Maximum=<-1500 -200 200>)
     2=?
     3=?
     Index=$User.GameModeID
     Num=::NewEnumerator0
    ))
   &ShapeLocation(MaskSpawnGroup=false, Override Seed=false, Sphere Radius=4, UseOwnerScale=true)
   /VehicleCosmetics/FX/VC_FXParents/Modules/DLR_ProductPreviews(GameModeID:NiagaraInt32=$System.GameModeID, bPreviewVehicleSpeed:NiagaraBool=$System.bPreviewVehicleSpeed)
  ]
  particleUpdateScript [
   &ParticleState()
   &SpriteRotationRate(Rotation Rate=&RandomRangeFloat(Minimum=350, Maximum=500))
   &AccelerationForce(Acceleration=&RandomRangeVector(Minimum=<0 0 -256>, Maximum=<0 0 -128>))
   &Collision(
    Age Colliding Particles=false
    Friction:NiagaraFloat=&RandomRangeFloat(Minimum=0.2, Maximum=0.5)
    KillOnCollision=false
    Module.Control Rest State Via Surface Normal=false
    Module.EnableMaxCollisionCount=false
    Module.Randomize Collision Normal Vector=false
   )
   &GravityForce()
   &InheritSourceMovement(
    Applied Position Scale=&Multiply_VectorByFloat(
     Vector=<1 1 1>
     Float=&FloatFromCurve(FloatCurve=DataInterface() {propertyValues {Curve {
         defaultValue=1
         keys [
          {arriveTangent=-1.7885, interpMode=RCIM_Cubic, leaveTangent=-1.7885, tangentMode=RCTM_User, time=0.05, value=1}
          {interpMode=RCIM_Cubic, tangentMode=RCTM_User, time=1, value=0}
         ]
        }}})
    )
    ApplyForce=false
    ApplyPosition=true
    ApplyPositionFalloff=false
   ) {module=InheritSourceMovement001}
   &SubUVAnimation(
    "Sprite Renderer "=DataInterface() {propertyValues {SpriteRenderer=&self:NE_VC_Boost_Sprite_Parent001_0.NiagaraSpriteRendererProperties_3}}
    UseEndFrame=false
    UseStartFrame=false
   )
   &SolveForcesAndVelocity(Clamp Velocity=false, Limit Acceleration=false)
  ]
  rendererClasses [&NiagaraSpriteRendererProperties]
  renderers [{rendererClass=&NiagaraSpriteRendererProperties}]
 }
 SpriteStars {
  bEnabled=false
  emitterSpawnScript [/SetVariables(Emitter.HighLightintensity:Vector3f=&SelectVectorFromArray(
     Direct Array Index=$System.ColorVariantID
     Vector Selection Array=DataInterface() {propertyValues {FloatData [
        <15 1 1>
        <0 0 0>
        <100 1 1>
        <30 1 1>
        <30 1 1>
        <30 1 1>
        <30 1 1>
        <15 1 1>
        <15 1 1>
        <15 1 1>
        <15 1 1>
        <20 1 1>
        <15 1 1>
        <15 1 1>
        <15 1 1>
       ]}}
    )) {isSetParameters=true, module=SetVariables_#5}]
  emitterUpdateScript [
   &EmitterState()
   &VC_FilteredSocketCount()
   &SpawnRate(SpawnRate=50, Use Spawn Probability=false)
   &SpawnRate(SpawnGroup=1, SpawnRate=25, Use Spawn Probability=false) {module=SpawnRate001}
  ]
  particleSpawnScript [
   &VC_PaintedVariantsOverride(
    Black=<rgba 0.001 0.001 0.001 1>
    BurntSienna=<rgba 0.1971 0.0609 0.0243 1>
    Cobalt=<rgba 0 0.008 0.4323 1>
    Crimson=<rgba 0.3003 0 0.0081 1>
    DefaultColor=<rgba 0.2708 0.2179 0.0324 1>
    ForestGreen=<rgba 0 0.2344 0.0118 1>
    Grey=<rgba 0.014 0.014 0.014 1>
    Lime=<rgba 0.1623 0.3247 0 1>
    Orange=<rgba 0.1753 0.0673 0.0094 1>
    Pink=<rgba 0.151 0.0181 0.1027 1>
    Purple=<rgba 0.0734 0 0.1788 1>
    SkyBlue=<rgba 0 0.1259 0.1892 1>
    TitaniumWhite=<rgba 0.2309 0.2309 0.2309 1>
   )
   &InitializeParticle(
    Color Mode=::NewEnumerator0
    Lifetime=&MergeHelper_FloatSwitch(
     0=&RandomRangeFloat(Minimum=0.4, Maximum=0.9)
     1=&RandomRangeFloat(Minimum=0.5, Maximum=0.9)
     2=?
     3=?
     Index=$User.GameModeID
     Num=::NewEnumerator0
    )
    Sprite Rotation Angle=&RandomRangeFloat(Minimum=0, Maximum=1)
    Sprite Rotation Mode=::NewEnumerator1
    Sprite Size Mode=::NewEnumerator3
    Uniform Sprite Size=&RandomRangeFloat(Minimum=25, Maximum=120)
    UsePositionOffset=false
    Write Material Random=true
   )
   &PartitionParticles(Number Of Partitions=$Emitter.PortCount)
   &SkeletalMeshLocation(
    Bone / Socket Index=$Particles.PartitionParticles.Partition
    Bone / Socket Sampling Mode=::NewEnumerator5
    MaskSpawnGroup=false
    Mesh Sampling Type=::NewEnumerator6
    Skeletal Mesh=$System.SkeletalMesh
   )
   &AddVelocity(Velocity=&MergeHelper_VectorSwitch(
     0=&RandomRangeVector(Minimum=<-900 0 100>, Maximum=<-1000 1 0>)
     1=&RandomRangeVector(Minimum=<-1000 0 100>, Maximum=<-1750 0 120>)
     2=&RandomRangeVector(Minimum=<-1000 0 0>, Maximum=<-1100 1 0>)
     3=?
     Index=$User.GameModeID
     Num=::NewEnumerator1
    ))
   &AddVelocity(Velocity=&RandomRangeVector(Minimum=<1 -25 60>, Maximum=<5 25 50>), Velocity Speed Scale=6) {module=AddVelocity001}
   &ShapeLocation(MaskSpawnGroup=false, Override Seed=false, Sphere Radius=30, UseOwnerScale=true)
   /SetVariables(Particles.VisibilityTag=$Emitter.SpawnGroup) {isSetParameters=true, module=SetVariables_#6}
  ]
  particleUpdateScript [
   &ParticleState()
   &self:WriteDeath_01(Position Array:NiagaraDataInterfaceArrayPosition=$System.DeathArray) {module=WriteDeath_01}
   &AccelerationForce(Acceleration=&RandomRangeVector(Minimum=<0 0 -1600>, Maximum=<0 0 -2048>))
   &Collision(
    Age Colliding Particles=false
    Analytical Collision Plane Position 1:Vector3f=&MergeHelper_VectorSwitch(0=<0 0 10>, 1=<0 0 380>, 2=?, 3=?, Index=$User.GameModeID, Num=::NewEnumerator0)
    Analytical Collision Plane Space:ENiagaraCoordinateSpace=::World
    CPU Collision Type=::NewEnumerator3
    KillOnCollision=false
    Module.Control Rest State Via Surface Normal=false
    Module.EnableMaxCollisionCount=false
    Module.Randomize Collision Normal Vector=false
    Perform 2 Analytical Collision Queries=false
   )
   &InheritSourceMovement(
    Applied Position Scale=&Multiply_VectorByFloat(
     Vector=<1 1 1>
     Float=&FloatFromCurve(FloatCurve=DataInterface() {propertyValues {Curve {keys [
          {arriveTangent=0.0056, interpMode=RCIM_Cubic, leaveTangent=0.0056, tangentMode=RCTM_User, time=0.1, value=1}
          {arriveTangent=-0.248, interpMode=RCIM_Cubic, leaveTangent=-0.248, tangentMode=RCTM_User, time=1, value=0.5}
         ]}}})
    )
    ApplyForce=false
    ApplyPosition=true
    ApplyPositionFalloff=false
   )
   &SolveForcesAndVelocity(Clamp Velocity=false, Limit Acceleration=false)
   &SubUVAnimation(
    "Sprite Renderer "=DataInterface() {propertyValues {SpriteRenderer=&self:SpriteStars_0.NiagaraSpriteRendererProperties_0}}
    UseEndFrame=false
    UseStartFrame=false
   )
   &DynamicMaterialParameters(
    Index 0 Param 1=1
    Index 0 Param 2=1.8
    Index 0 Param 3=1
    Module.Index0Float0Write=true
    Module.Index0Float1Write=true
    Module.Index0Float2Write=true
    Module.Index0Float3Write=true
   )
   &ScaleColor(
    Scale RGB=&Multiply_VectorByFloat(
     Vector=<1 1 1>
     Float=&FloatFromCurve(FloatCurve=DataInterface() {propertyValues {Curve {keys [<at 0 8>,<at 0.15 2>]}}})
    )
    ScaleA=true
    ScaleRGB=true
   )
  ]
  rendererClasses [&NiagaraSpriteRendererProperties]
  renderers [
   {rendererClass=&NiagaraSpriteRendererProperties}
   {rendererClass=&NiagaraSpriteRendererProperties, rendererIndex=1}
  ]
 }
 SpriteStars001 {
  emitterSpawnScript [/SetVariables(Emitter.HighLightintensity:Vector3f=&SelectVectorFromArray(
     Direct Array Index=$System.ColorVariantID
     Vector Selection Array=DataInterface() {propertyValues {FloatData [
        <15 1 1>
        <0 0 0>
        <100 1 1>
        <30 1 1>
        <30 1 1>
        <30 1 1>
        <30 1 1>
        <15 1 1>
        <15 1 1>
        <15 1 1>
        <15 1 1>
        <20 1 1>
        <15 1 1>
        <15 1 1>
        <15 1 1>
       ]}}
    )) {isSetParameters=true, module=SetVariables_#5}]
  emitterUpdateScript [
   &EmitterState()
   &VC_FilteredSocketCount()
   &SpawnRate(SpawnRate=50, Use Spawn Probability=false)
   &SpawnRate(SpawnGroup=1, SpawnRate=25, Use Spawn Probability=false) {module=SpawnRate002}
  ]
  particleSpawnScript [
   &VC_PaintedVariantsOverride(
    Black=<rgba 0.001 0.001 0.001 1>
    BurntSienna=<rgba 0.1971 0.0609 0.0243 1>
    Cobalt=<rgba 0 0.008 0.4323 1>
    Crimson=<rgba 0.3003 0 0.0081 1>
    DefaultColor=<rgba 0.2708 0.2179 0.0324 1>
    ForestGreen=<rgba 0 0.2344 0.0118 1>
    Grey=<rgba 0.014 0.014 0.014 1>
    Lime=<rgba 0.1623 0.3247 0 1>
    Orange=<rgba 0.1753 0.0673 0.0094 1>
    Pink=<rgba 0.151 0.0181 0.1027 1>
    Purple=<rgba 0.0734 0 0.1788 1>
    SkyBlue=<rgba 0 0.1259 0.1892 1>
    TitaniumWhite=<rgba 0.2309 0.2309 0.2309 1>
   )
   &InitializeParticle(
    Color Mode=::NewEnumerator0
    Lifetime=&MergeHelper_FloatSwitch(
     0=&RandomRangeFloat(Minimum=0.4, Maximum=0.9)
     1=&RandomRangeFloat(Minimum=0.5, Maximum=0.9)
     2=&RandomRangeFloat(Minimum=0.4, Maximum=0.9)
     3=&RandomRangeFloat(Minimum=0.2, Maximum=0.6)
     Index=$User.GameModeID
     Num=::NewEnumerator2
    )
    Sprite Rotation Angle=&RandomRangeFloat(Minimum=0, Maximum=1)
    Sprite Rotation Mode=::NewEnumerator1
    Sprite Size Mode=::NewEnumerator3
    Uniform Sprite Size=&RandomRangeFloat(Minimum=25, Maximum=120)
    UsePositionOffset=false
    Write Material Random=true
   )
   &PartitionParticles(Number Of Partitions=$Emitter.PortCount)
   &SkeletalMeshLocation(
    Bone / Socket Index=$Particles.PartitionParticles.Partition
    Bone / Socket Sampling Mode=::NewEnumerator5
    MaskSpawnGroup=false
    Mesh Sampling Type=::NewEnumerator6
    Sampled Position Offset=&AddVectorToPosition(
     Position=&MergeHelper_VectorSwitch(0=<0 0 0>, 1=<0 0 0>, 2=<0 0 0>, 3=<-50 0 0>, Index=$User.GameModeID, Num=::NewEnumerator2)
     Vector=Linked() {linkedVariable {name=User.DTCarOffset, type=NiagaraPosition}}
    )
    Skeletal Mesh=$System.SkeletalMesh
   )
   &AddVelocity(Velocity=&MergeHelper_VectorSwitch(
     0=&RandomRangeVector(Minimum=<-900 0 100>, Maximum=<-1000 1 0>)
     1=&RandomRangeVector(Minimum=<-1000 0 100>, Maximum=<-1750 0 120>)
     2=&RandomRangeVector(Minimum=<-1000 0 0>, Maximum=<-1100 1 0>)
     3=&RandomRangeVector(Minimum=<-800 0 50>, Maximum=<-1000 0 0>)
     Index=$User.GameModeID
     Num=::NewEnumerator2
    ))
   &AddVelocity(Velocity=&RandomRangeVector(Minimum=<1 -25 60>, Maximum=<5 25 50>), Velocity Speed Scale=6) {module=AddVelocity001}
   &ShapeLocation(MaskSpawnGroup=false, Override Seed=false, Sphere Radius=30, UseOwnerScale=true)
   /SetVariables(Particles.VisibilityTag=$Emitter.SpawnGroup) {isSetParameters=true, module=SetVariables_#6}
   &DLR_ProductPreviews(GameModeID:NiagaraInt32=$System.GameModeID, bPreviewVehicleSpeed:NiagaraBool=$System.bPreviewVehicleSpeed)
  ]
  particleUpdateScript [
   &ParticleState()
   &self:WriteDeath_01(Position Array:NiagaraDataInterfaceArrayPosition=$System.DeathArray) {module=WriteDeath_01}
   &AccelerationForce(Acceleration=&RandomRangeVector(Minimum=<0 0 -1600>, Maximum=<0 0 -2048>))
   &Collision(
    Age Colliding Particles=false
    Analytical Collision Plane Position 1:Vector3f=&MergeHelper_VectorSwitch(0=<0 0 10>, 1=<0 0 380>, 2=?, 3=?, Index=$User.GameModeID, Num=::NewEnumerator0)
    Analytical Collision Plane Space:ENiagaraCoordinateSpace=::World
    CPU Collision Type=::NewEnumerator3
    KillOnCollision=false
    Module.Control Rest State Via Surface Normal=false
    Module.EnableMaxCollisionCount=false
    Module.Randomize Collision Normal Vector=false
    Perform 2 Analytical Collision Queries=false
   )
   &InheritSourceMovement(
    Applied Position Scale=&Multiply_VectorByFloat(
     Vector=<1 1 1>
     Float=&FloatFromCurve(FloatCurve=DataInterface() {propertyValues {Curve {keys [
          {arriveTangent=0.0056, interpMode=RCIM_Cubic, leaveTangent=0.0056, tangentMode=RCTM_User, time=0.1, value=1}
          {arriveTangent=-0.248, interpMode=RCIM_Cubic, leaveTangent=-0.248, tangentMode=RCTM_User, time=1, value=0.5}
         ]}}})
    )
    ApplyForce=false
    ApplyPosition=true
    ApplyPositionFalloff=false
   ) {enabled=false}
   &InheritSourceMovement(
    Applied Position Scale=&Multiply_VectorByFloat(
     Vector=<1 1 1>
     Float=&FloatFromCurve(FloatCurve=DataInterface() {propertyValues {Curve {
         defaultValue=1
         keys [
          {arriveTangent=-1.7885, interpMode=RCIM_Cubic, leaveTangent=-1.7885, tangentMode=RCTM_User, time=0.05, value=1}
          {interpMode=RCIM_Cubic, tangentMode=RCTM_User, time=1, value=0}
         ]
        }}})
    )
    ApplyForce=false
    ApplyPosition=true
    ApplyPositionFalloff=false
   ) {module=InheritSourceMovement001}
   &DynamicMaterialParameters(
    Index 0 Param 1=1
    Index 0 Param 2=1.8
    Index 0 Param 3=1
    Module.Index0Float0Write=true
    Module.Index0Float1Write=true
    Module.Index0Float2Write=true
    Module.Index0Float3Write=true
   )
   &SubUVAnimation(
    "Sprite Renderer "=DataInterface() {propertyValues {SpriteRenderer=&self:SpriteStars001_0.NiagaraSpriteRendererProperties_4}}
    UseEndFrame=false
    UseStartFrame=false
   )
   &ScaleColor(
    Scale RGB=&Multiply_VectorByFloat(
     Vector=<1 1 1>
     Float=&FloatFromCurve(FloatCurve=DataInterface() {propertyValues {Curve {keys [<at 0 8>,<at 0.15 2>]}}})
    )
    ScaleA=true
    ScaleRGB=true
   )
   &SolveForcesAndVelocity(Clamp Velocity=false, Limit Acceleration=false)
  ]
  rendererClasses [&NiagaraSpriteRendererProperties]
  renderers [
   {rendererClass=&NiagaraSpriteRendererProperties}
   {rendererClass=&NiagaraSpriteRendererProperties, rendererIndex=1}
  ]
 }
 ribbon {
  emitterUpdateScript [
   &EmitterState(Inactive Response=::NewEnumerator1, Life Cycle Mode=::NewEnumerator1, Loop Duration=1, UseLoopDelay=false)
   &VC_FilteredSocketCount()
   &VC_RateQualityHandler(Epic Spawn Rate=45, Low Spawn Rate=30, Use Custom Rate?=true) {enabled=false}
   &VC_SpawnRibbon_v1(
    Number of Ribbons=$Emitter.PortCount
    SpawnRate=&MergeHelper_FloatSwitch(0=60, 1=60, 2=60, 3=20, Index=$User.GameModeID, Num=::NewEnumerator2)
   )
  ]
  particleSpawnScript [
   &VC_PaintedVariantsOverride(
    Black=<rgba 0.001 0.001 0.001 1.5>
    BurntSienna=<rgba 0.2865 0.1227 0.0632 1>
    Cobalt=<rgba 0.0993 0.1409 0.724 1.5>
    Crimson=<rgba 0.4635 0.0427 0.0541 1>
    DefaultColor=<rgba 1 0.6 0.1 1>
    ForestGreen=<rgba 0.0477 0.5 0.0706 1>
    Grey=<rgba 0.0747 0.0747 0.0747 1>
    Lime=<rgba 0.3142 0.5885 0.0398 1>
    Orange=<rgba 0.4635 0.1617 0 1>
    Pink=<rgba 0.3003 0.0771 0.2854 1>
    Purple=<rgba 0.245 0.0517 0.5226 1>
    Saffron=<rgba 0.12 0.08 0 1.25>
    SkyBlue=<rgba 0.1034 0.3915 0.5365 1>
    TitaniumWhite=<rgba 0.6615 0.6615 0.6615 1>
   )
   &VC_InitializeRibbon_v1(
    Color=$Particles.Color
    Custom Offset=true
    Enable Skeletal Position=true
    Lifetime=&MergeHelper_FloatSwitch(0=0.75, 1=0.75, 2=0.75, 3=1, Index=$User.GameModeID, Num=::NewEnumerator2)
    Local Vector Offset=&AddVectorToPosition(Position=<-30 0 0>, Vector=$User.DTCarOffset)
    Offset Amount=1
    Ribbon Width=40
    Write Material Random=false
   )
   &AddVelocity(Velocity=&MergeHelper_VectorSwitch(0=<-1250 0 0>, 1=<-1250 0 0>, 2=<-1250 0 0>, 3=<-2400 0 0>, Index=$User.GameModeID, Num=::NewEnumerator2)) {module=AddVelocity001}
   &DLR_ProductPreviews(GameModeID:NiagaraInt32=$System.GameModeID, bPreviewVehicleSpeed:NiagaraBool=$System.bPreviewVehicleSpeed)
  ]
  particleUpdateScript [
   &ParticleState()
   &InheritSourceMovement(
    Applied Position Scale=&Multiply_VectorByFloat(Vector=<1 1 1>, Float=1)
    ApplyForce=false
    ApplyPosition=true
    ApplyPositionFalloff=false
   ) {enabled=false, module=InheritSourceMovement001}
   &CurlNoiseForce(Bias Noise=false, Noise Frequency=500, Noise Quality / Cost=::NewEnumerator3, Noise Strength=800, Pan Noise=true) {enabled=false}
   &InheritSourceMovement(
    Applied Position Scale=&VectorFromFloat(Value=&FloatFromCurve(FloatCurve=DataInterface() {propertyValues {Curve {
         defaultValue=1
         keys [
          {arriveTangent=-1.7885, interpMode=RCIM_Cubic, leaveTangent=-1.7885, tangentMode=RCTM_User, time=0.05, value=1}
          {interpMode=RCIM_Cubic, tangentMode=RCTM_User, time=1, value=0}
         ]
        }}}))
    ApplyForce=false
    ApplyPosition=true
    ApplyPositionFalloff=false
   )
   &SolveForcesAndVelocity(Clamp Velocity=false, Limit Acceleration=false)
   &ScaleRibbonWidth(Ribbon Width Scale=&FloatFromCurve(FloatCurve=DataInterface() {propertyValues {Curve {keys [<at 0 1>,<at 1 0.1>]}}}))
   &VC_UpdateRibbon_v1(Enable Skeletal Position=true)
  ]
  rendererClasses [&NiagaraRibbonRendererProperties]
  renderers [{rendererClass=&NiagaraRibbonRendererProperties}]
 }
 ribbon001 {
  emitterUpdateScript [
   &EmitterState(Inactive Response=::NewEnumerator1, Life Cycle Mode=::NewEnumerator1, Loop Duration=1, UseLoopDelay=false)
   &VC_FilteredSocketCount()
   &VC_RateQualityHandler(Epic Spawn Rate=45, Low Spawn Rate=30, Use Custom Rate?=true) {enabled=false}
   &VC_SpawnRibbon_v1(
    Number of Ribbons=$Emitter.PortCount
    SpawnRate=&MergeHelper_FloatSwitch(0=60, 1=60, 2=60, 3=20, Index=$User.GameModeID, Num=::NewEnumerator2)
   )
  ]
  particleSpawnScript [
   &VC_PaintedVariantsOverride(
    Black=<rgba 0.001 0.001 0.001 1.5>
    BurntSienna=<rgba 0.2865 0.1227 0.0632 1>
    Cobalt=<rgba 0.0993 0.1409 0.724 1.5>
    Crimson=<rgba 0.4635 0.0427 0.0541 1>
    DefaultColor=<rgba 1 0.6 0.1 1>
    ForestGreen=<rgba 0.0477 0.5 0.0706 1>
    Grey=<rgba 0.0747 0.0747 0.0747 1>
    Lime=<rgba 0.3132 0.5885 0.0378 1>
    Orange=<rgba 0.4635 0.1617 0 1>
    Pink=<rgba 0.3003 0.0771 0.2854 1>
    Purple=<rgba 0.245 0.0517 0.5226 1>
    Saffron=<rgba 0.12 0.08 0 1.25>
    SkyBlue=<rgba 0.1034 0.3915 0.5365 1>
    TitaniumWhite=<rgba 0.6615 0.6615 0.6615 1>
   )
   &VC_InitializeRibbon_v1(
    Color=$Particles.Color
    Custom Offset=true
    Enable Skeletal Position=true
    Lifetime=&MergeHelper_FloatSwitch(0=0.75, 1=0.75, 2=0.75, 3=1, Index=$User.GameModeID, Num=::NewEnumerator2)
    Local Vector Offset=&AddVectorToPosition(Position=<-30 0 0>, Vector=$User.DTCarOffset)
    Offset Amount=1
    Ribbon Width=40
    Write Material Random=false
   )
   &AddVelocity(Velocity=&MergeHelper_VectorSwitch(0=<-1250 0 0>, 1=<-1250 0 0>, 2=<-1250 0 0>, 3=<-2400 0 0>, Index=$User.GameModeID, Num=::NewEnumerator2)) {module=AddVelocity001}
   &DLR_ProductPreviews(GameModeID:NiagaraInt32=$System.GameModeID, bPreviewVehicleSpeed:NiagaraBool=$System.bPreviewVehicleSpeed)
  ]
  particleUpdateScript [
   &ParticleState()
   &InheritSourceMovement(
    Applied Position Scale=&Multiply_VectorByFloat(Vector=<1 1 1>, Float=1)
    ApplyForce=false
    ApplyPosition=true
    ApplyPositionFalloff=false
   ) {enabled=false, module=InheritSourceMovement001}
   &CurlNoiseForce(Bias Noise=false, Noise Frequency=500, Noise Quality / Cost=::NewEnumerator3, Noise Strength=800, Pan Noise=true, Pan Noise Field=<0 50 0>)
   &InheritSourceMovement(
    Applied Position Scale=&VectorFromFloat(Value=&FloatFromCurve(FloatCurve=DataInterface() {propertyValues {Curve {
         defaultValue=1
         keys [
          {arriveTangent=-1.7885, interpMode=RCIM_Cubic, leaveTangent=-1.7885, tangentMode=RCTM_User, time=0.05, value=1}
          {interpMode=RCIM_Cubic, tangentMode=RCTM_User, time=1, value=0}
         ]
        }}}))
    ApplyForce=false
    ApplyPosition=true
    ApplyPositionFalloff=false
   )
   &SolveForcesAndVelocity(Clamp Velocity=false, Limit Acceleration=false)
   &ScaleRibbonWidth(Ribbon Width Scale=&FloatFromCurve(FloatCurve=DataInterface() {propertyValues {Curve {keys [<at 0 1>,<at 1 0.1>]}}}))
   &VC_UpdateRibbon_v1(Enable Skeletal Position=true)
  ]
  rendererClasses [&NiagaraRibbonRendererProperties]
  renderers [{rendererClass=&NiagaraRibbonRendererProperties}]
 }
 ribbon002 {
  bEnabled=false
  emitterUpdateScript [
   &EmitterState(Inactive Response=::NewEnumerator1, Life Cycle Mode=::NewEnumerator1, Loop Duration=1, UseLoopDelay=false)
   &VC_FilteredSocketCount()
   &VC_RateQualityHandler(Epic Spawn Rate=45, Low Spawn Rate=30, Use Custom Rate?=true) {enabled=false}
   &VC_SpawnRibbon_v1(Number of Ribbons=$Emitter.PortCount, SpawnRate=60)
  ]
  particleSpawnScript [
   &VC_PaintedVariantsOverride(
    Black=<rgba 0.001 0.001 0.001 1.5>
    BurntSienna=<rgba 0.2865 0.1227 0.0632 1>
    Cobalt=<rgba 0.0993 0.1409 0.724 1.5>
    Crimson=<rgba 0.4635 0.0427 0.0541 1>
    DefaultColor=<rgba 1 0.6 0.1 1>
    ForestGreen=<rgba 0.0477 0.5 0.0706 1>
    Grey=<rgba 0.0747 0.0747 0.0747 1>
    Lime=<rgba 0.3132 0.5885 0.0378 1>
    Orange=<rgba 0.4635 0.1617 0 1>
    Pink=<rgba 0.3003 0.0771 0.2854 1>
    Purple=<rgba 0.245 0.0517 0.5226 1>
    Saffron=<rgba 0.12 0.08 0 1.25>
    SkyBlue=<rgba 0.1034 0.3915 0.5365 1>
    TitaniumWhite=<rgba 0.6615 0.6615 0.6615 1>
   )
   &VC_InitializeRibbon_v1(
    Color=$Particles.Color
    Custom Offset=true
    Enable Skeletal Position=true
    Lifetime=0.75
    Local Vector Offset=&AddVectorToPosition(Position=<-30 0 0>, Vector=$User.DTCarOffset)
    Offset Amount=1
    Ribbon Width=40
    Write Material Random=false
   )
   &AddVelocity(Velocity=<1250 0 0>, Velocity Speed Scale=-1) {module=AddVelocity001}
  ]
  particleUpdateScript [
   &ParticleState()
   &InheritSourceMovement(
    Applied Position Scale=&Multiply_VectorByFloat(Vector=<1 1 1>, Float=1)
    ApplyForce=false
    ApplyPosition=true
    ApplyPositionFalloff=false
   )
   &CurlNoiseForce(Bias Noise=false, Noise Frequency=500, Noise Quality / Cost=::NewEnumerator3, Noise Strength=800, Pan Noise=true, Pan Noise Field=<0 50 0>)
   &SolveForcesAndVelocity(Clamp Velocity=false, Limit Acceleration=false)
   &VC_UpdateRibbon_v1(Enable Skeletal Position=true)
   &ScaleRibbonWidth(Ribbon Width Scale=&FloatFromCurve(FloatCurve=DataInterface() {propertyValues {Curve {keys [<at 0 1>,<at 1 0.1>]}}}))
  ]
  rendererClasses [&NiagaraRibbonRendererProperties]
  renderers [{rendererClass=&NiagaraRibbonRendererProperties}]
 }
 ribbon003 {
  bEnabled=false
  emitterUpdateScript [
   &EmitterState(Inactive Response=::NewEnumerator1, Life Cycle Mode=::NewEnumerator1, Loop Duration=1, UseLoopDelay=false)
   &VC_FilteredSocketCount()
   &VC_RateQualityHandler(Epic Spawn Rate=45, Low Spawn Rate=30, Use Custom Rate?=true) {enabled=false}
   &VC_SpawnRibbon_v1(Number of Ribbons=$Emitter.PortCount, SpawnRate=60)
  ]
  particleSpawnScript [
   &VC_PaintedVariantsOverride(
    Black=<rgba 0.001 0.001 0.001 1.5>
    BurntSienna=<rgba 0.2865 0.1227 0.0632 1>
    Cobalt=<rgba 0.0993 0.1409 0.724 1.5>
    Crimson=<rgba 0.4635 0.0427 0.0541 1>
    DefaultColor=<rgba 1 0.6 0.1 1>
    ForestGreen=<rgba 0.0477 0.5 0.0706 1>
    Grey=<rgba 0.0747 0.0747 0.0747 1>
    Lime=<rgba 0.3142 0.5885 0.0398 1>
    Orange=<rgba 0.4635 0.1617 0 1>
    Pink=<rgba 0.3003 0.0771 0.2854 1>
    Purple=<rgba 0.245 0.0517 0.5226 1>
    Saffron=<rgba 0.12 0.08 0 1.25>
    SkyBlue=<rgba 0.1034 0.3915 0.5365 1>
    TitaniumWhite=<rgba 0.6615 0.6615 0.6615 1>
   )
   &VC_InitializeRibbon_v1(
    Color=$Particles.Color
    Custom Offset=true
    Enable Skeletal Position=true
    Lifetime=0.75
    Local Vector Offset=&AddVectorToPosition(Position=<-30 0 0>, Vector=$User.DTCarOffset)
    Offset Amount=1
    Ribbon Width=40
    Write Material Random=false
   )
   &AddVelocity(Velocity=<1250 0 0>, Velocity Speed Scale=-1) {module=AddVelocity001}
  ]
  particleUpdateScript [
   &ParticleState()
   &InheritSourceMovement(
    Applied Position Scale=&Multiply_VectorByFloat(Vector=<1 1 1>, Float=1)
    ApplyForce=false
    ApplyPosition=true
    ApplyPositionFalloff=false
   )
   &CurlNoiseForce(Bias Noise=false, Noise Frequency=500, Noise Quality / Cost=::NewEnumerator3, Noise Strength=800, Pan Noise=true)
   &SolveForcesAndVelocity(Clamp Velocity=false, Limit Acceleration=false)
   &VC_UpdateRibbon_v1(Enable Skeletal Position=true)
   &ScaleRibbonWidth(Ribbon Width Scale=&FloatFromCurve(FloatCurve=DataInterface() {propertyValues {Curve {keys [<at 0 1>,<at 1 0.1>]}}}))
  ]
  rendererClasses [&NiagaraRibbonRendererProperties]
  renderers [{rendererClass=&NiagaraRibbonRendererProperties}]
 }
}
summary {
 systemName=NS_Boost_Stars
 emitters [
  {emitterName=ribbon001, rendererClasses [&NiagaraRibbonRendererProperties]}
  {emitterName=ribbon, rendererClasses [&NiagaraRibbonRendererProperties]}
  {emitterName=NE_VC_Boost_Sprite_Parent001, rendererClasses [&NiagaraSpriteRendererProperties]}
  {emitterName=SpriteStars001, rendererClasses [&NiagaraSpriteRendererProperties]}
  {bEnabled=false, emitterName=ribbon002, rendererClasses [&NiagaraRibbonRendererProperties]}
  {bEnabled=false, emitterName=ribbon003, rendererClasses [&NiagaraRibbonRendererProperties]}
  {bEnabled=false, emitterName=NE_VC_Boost_Sprite_Parent, rendererClasses [&NiagaraSpriteRendererProperties]}
  {bEnabled=false, emitterName=SpriteStars, rendererClasses [&NiagaraSpriteRendererProperties]}
 ]
 userVariables [
  User.bIsCustomColor {type=NiagaraBool}
  User.bIsFrontendPreview {type=NiagaraBool}
  User.bIsPainted {type=NiagaraBool}
  User.blsIsCoroColor {type=NiagaraBool}
  User.bPreviewVehicleSpeed {type=NiagaraBool}
  User.ColorVariantID {type=NiagaraInt32}
  User.CustomColor {type=LinearColor, description="", defaultValue {struct=LinearColor, value=<rgba 5 0 4.8885 1>}}
  User.DTCarOffset {type=NiagaraPosition, description="", defaultValue {}}
  User.DTCarOffset {type=Vector3f, description="", defaultValue {struct=Vector3f, value=<0 0 0>}}
  User.GameModeID {type=NiagaraInt32, defaultValue=3}
  User.PaintedColor {type=LinearColor, description="", defaultValue {struct=LinearColor, value=<rgba 1 1 1 1>}}
  User.PaintedScalar {type=NiagaraFloat, defaultValue=1, description=""}
  User.PortCount {type=NiagaraInt32, defaultValue=2}
 ]
}
